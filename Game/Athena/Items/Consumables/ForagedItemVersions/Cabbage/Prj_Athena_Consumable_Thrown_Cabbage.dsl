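; Thrown cabbage projectile for the foraged consumable, which reports long throws to the quest system.
; BeginPlay stores the projectile's starting location, and OnPickupSpawned stores where the spawned pickup lands.
; If the straight-line distance between them is at least 10,000 units (100 m), it sends a ThrowCabbage custom stat event to the instigating player's quest manager.
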(event EventBeginPlay
  (|CallParentFunction)
  (Variables|Default|SetInitial%20Loc (Projectile|GetCurrentLocation)))

(event EventOnPickupSpawned
  (params (Spawned%20Pickup :type Fort%20Pickup%20Object%20Reference))
  (|CallParentFunction Spawned%20Pickup)
  (Variables|Default|SetLanded%20Loc (Transformation|GetActorLocation Spawned%20Pickup))
  (if (>= (Math|Vector|Distance(Vector) (Variables|Default|GetInitial%20Loc) (Variables|Default|GetLanded%20Loc)) 1e+04)
    (Quest|SendComplexCustomStatEvent (Quests|GetQuestManager (Projectile|GetInstigatorPlayerController) "Invalid") 0 "(GameplayTags=((TagName=\"Athena.Quests.ThrowCabbage\")))" "(GameplayTags=)" 1)))

(fn ConstructionScript
  (|CallParentFunction))
